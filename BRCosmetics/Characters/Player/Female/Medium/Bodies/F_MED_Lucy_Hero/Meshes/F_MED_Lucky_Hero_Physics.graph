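# Physics asset for the Lucy Hero medium female player cosmetic body, covering only the dynamic armband bones for secondary motion.
# Small capsule bodies form a knot and two two-segment strands, high and low, all rooted at the knot master bone.
# Each strand's second segment is constrained to its first, with moderate swing limits and tight twist limits so the pieces can sway without spinning.
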PhysicsAsset /BRCosmetics/Characters/Player/Female/Medium/Bodies/F_MED_Lucy_Hero/Meshes/F_MED_Lucky_Hero_Physics
bodies {
 dyn_armband_hi_01 [{shapeType=Capsule, length=2, radius=1, center=<0.8807 0 0>, rotation=<rot 90 0 0>}]
 dyn_armband_hi_02 [{shapeType=Capsule, length=1, radius=1, center=<0.5414 -0.0261 -0.1905>, rotation=<rot 90 0 0>}]
 dyn_armband_knot [{shapeType=Capsule, length=1.004, radius=0.9876, rotation=<rot 90 0 0>}]
 dyn_armband_knot_MSTR [{shapeType=Capsule, length=0.01, radius=0.5, rotation=<rot 0 -90 0>}]
 dyn_armband_low_01 [{shapeType=Capsule, length=2, radius=1, center=<0.8807 0 0>, rotation=<rot 90 0 0>}]
 dyn_armband_low_02 [{shapeType=Capsule, length=2.5, radius=1, center=<0.7577 0.0001 -0.0886>, rotation=<rot 83.3318 -0.0002 -0.0001>}]
}
constraints [
 {bone1Name=dyn_armband_knot, bone2Name=dyn_armband_knot_MSTR, swing1LimitDegrees=20, swing2LimitDegrees=20, twistLimitDegrees=3}
 {bone1Name=dyn_armband_hi_01, bone2Name=dyn_armband_knot_MSTR, swing1LimitDegrees=25, swing2LimitDegrees=25, twistLimitDegrees=3}
 {bone1Name=dyn_armband_hi_02, bone2Name=dyn_armband_hi_01, swing1LimitDegrees=30, swing2LimitDegrees=30, twistLimitDegrees=5}
 {bone1Name=dyn_armband_low_01, bone2Name=dyn_armband_knot_MSTR, swing1LimitDegrees=30, swing2LimitDegrees=30, twistLimitDegrees=3}
 {bone1Name=dyn_armband_low_02, bone2Name=dyn_armband_low_01, swing1LimitDegrees=30, swing2LimitDegrees=30, twistLimitDegrees=3}
]
