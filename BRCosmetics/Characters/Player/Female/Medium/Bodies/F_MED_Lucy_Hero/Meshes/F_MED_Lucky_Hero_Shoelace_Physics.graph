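# Physics asset for the shoelace of the Lucy Hero female medium player body, defining the bodies and joints for dangling boot-string physics.
# It has three capsule bodies: a short base plus longer inner and outer right-side string bodies.
# Both string bodies are constrained directly to the base rather than chained, with moderate, asymmetric swing limits and a very tight twist limit.
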PhysicsAsset /BRCosmetics/Characters/Player/Female/Medium/Bodies/F_MED_Lucy_Hero/Meshes/F_MED_Lucky_Hero_Shoelace_Physics
bodies {
 dyn_boot_string_base [{shapeType=Capsule, length=2.5312, radius=2.7188, center=<0 3.1039 0>, rotation=<rot 90 0 0>}]
 dyn_boot_string_in_r [{shapeType=Capsule, length=6.2772, radius=2.4248, center=<4.9212 0.0001 -1.4801>, rotation=<rot 68.2267 0 0>}]
 dyn_boot_string_out_r [{shapeType=Capsule, length=5.1926, radius=2.5133, center=<3.8213 -0.0006 1.6384>, rotation=<rot 65.8002 180 180>}]
}
constraints [
 {bone1Name=dyn_boot_string_in_r, bone2Name=dyn_boot_string_base, swing1LimitDegrees=35, swing2LimitDegrees=20, twistLimitDegrees=5}
 {bone1Name=dyn_boot_string_out_r, bone2Name=dyn_boot_string_base, swing1LimitDegrees=35, swing2LimitDegrees=20, twistLimitDegrees=5}
]
