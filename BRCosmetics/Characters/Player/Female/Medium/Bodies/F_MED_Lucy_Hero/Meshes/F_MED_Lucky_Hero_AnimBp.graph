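# Animation Blueprint for the Lucky Hero female character body that follows the attached parent mesh's pose and adds secondary physics motion.
# After copying that pose, weighted constraints position the right boot-string base from calf bones and the armband knot from upper-arm bones.
# Each constraint feeds a RigidBody simulation in base-bone space: shoelace physics relative to root, then the main physics asset relative to spine_02.
# The event graph holds only unwired default nodes.
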AnimBlueprint /BRCosmetics/Characters/Player/Female/Medium/Bodies/F_MED_Lucy_Hero/Meshes/F_MED_Lucky_Hero_AnimBp
target_skeleton=&/F_MED_Lucky_Hero_Skeleton
graphs [
 AnimGraph {nodes [
   n1 = "Constraint - dyn_boot_string_base (PP)"!(
    ComponentPose=Local To Component(LocalPose=Copy Pose From Mesh() {bUseAttachedParent=true})
    Alpha=1
    ConstraintWeights_0=0.8257
    ConstraintWeights_1:float=0.1743
   ) -> (Pose) {
    alphaBoolBlend {blendOption=Linear}
    boneToModify {boneName=dyn_boot_string_base}
    constraintSetup [{targetBone {boneName=calf_twist_02_r}}, {targetBone {boneName=calf_r}}]
    constraintWeights [1,1]
   }
   n2 = RigidBody(ComponentPose=@n1) {
    bClampLinearTranslationLimitToRefPose=true
    bOverrideWorldGravity=true
    overridePhysicsAsset=&/F_MED_Lucky_Hero_Shoelace_Physics
    simulationSpace=BaseBoneSpace
    alphaBoolBlend {blendOption=Linear}
    alphaScaleBiasClamp {interpSpeedDecreasing=6, interpSpeedIncreasing=6}
    baseBoneRef {boneName=root}
    componentAppliedLinearAccClamp=<6000 6000 6000>
    componentLinearAccScale=<0.25 0.25 0.25>
    componentLinearVelScale=<0.25 0.25 0.25>
    overrideWorldGravity {z=-800}
   }
   n3 = "Constraint - dyn_armband_knot_MSTR (PPP)"!(ComponentPose=@n2, Alpha=1, ConstraintWeights_0=0.5897, ConstraintWeights_1:float=0.1979, ConstraintWeights_2:float=0.2123) -> (Pose) {
    alphaBoolBlend {blendOption=Linear}
    boneToModify {boneName=dyn_armband_knot_MSTR}
    constraintSetup [{targetBone {boneName=upperarm_twist_01_r}}, {targetBone {boneName=upperarm_r}}, {targetBone {boneName=upperarm_twist_02_r}}]
    constraintWeights [1,1,1]
   }
   n4 = RigidBody(ComponentPose=@n3) {
    bClampLinearTranslationLimitToRefPose=true
    bOverrideWorldGravity=true
    overridePhysicsAsset=&/F_MED_Lucky_Hero_Physics
    simulationSpace=BaseBoneSpace
    alphaBoolBlend {blendOption=Linear}
    baseBoneRef {boneName=spine_02}
    componentAppliedLinearAccClamp=<6000 6000 6000>
    componentLinearAccScale=<0.5 0.5 0.5>
    componentLinearVelScale=<0.5 0.5 0.5>
    overrideWorldGravity {z=-980}
   }
   Output Pose(Result=Component To Local(ComponentPose=@n4)) {layerGroup=None}
  ]}
 EventGraph {nodes [
   TryGetPawnOwner()
   Event BlueprintUpdateAnimation()
  ]}
]
variables []
