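# Physics asset for the face accessory part of the female medium Lucy Hero cosmetic body, covering dynamic ear and hair bones.
# A head capsule anchors four hair boxes, each constrained directly to it with tight swing limits and locked twist.
# Each ear is a box chain from base through parent, 01 and 02, with identical, looser limits on both sides.
# No constraint links the ear bases to the head.
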PhysicsAsset /BRCosmetics/Characters/Player/Female/Medium/Bodies/F_MED_Lucy_Hero/Meshes/Parts/F_MED_Lucky_Hero_FaceAcc_Physics
bodies {
 dyn_ear_l_01 [{shapeType=Box, extentX=3.9136, extentY=4.6719, extentZ=1.25, center=<2.113 0 0>}]
 dyn_ear_l_02 [{shapeType=Box, extentX=3.9136, extentY=4.6719, extentZ=1.25, center=<2.113 0 0>}]
 dyn_ear_l_base [{shapeType=Box, extentX=1.75, extentY=1.75, extentZ=1.75}]
 dyn_ear_l_parent [{shapeType=Box, extentX=6.8981, extentY=1.7356, extentZ=6.3629, center=<2.6048 0.3454 0.4865>, rotation=<rot 0 172.5929 -85.2956>}]
 dyn_ear_r_01 [{shapeType=Box, extentX=3.9136, extentY=4.6719, extentZ=1.25, center=<2.113 0 0>}]
 dyn_ear_r_02 [{shapeType=Box, extentX=3.9136, extentY=4.6719, extentZ=1.25, center=<2.113 0 0>}]
 dyn_ear_r_base [{shapeType=Box, extentX=1.75, extentY=1.75, extentZ=1.75}]
 dyn_ear_r_parent [{shapeType=Box, extentX=6.8981, extentY=1.7356, extentZ=6.3629, center=<2.6048 0.3454 0.4865>, rotation=<rot 0 172.5929 -85.2956>}]
 dyn_hair_01 [{shapeType=Box, extentX=2.3707, extentY=2.9549, extentZ=2.75, center=<0.7429 0 0>}]
 dyn_hair_02 [{shapeType=Box, extentX=1.3891, extentY=1.5279, extentZ=1.585, center=<0.9184 0 0>}]
 dyn_hair_03 [{shapeType=Box, extentX=2.7862, extentY=2.0564, extentZ=1.8489, center=<0.7782 0 0>}]
 dyn_hair_04 [{shapeType=Box, extentX=2.4759, extentY=2.3235, extentZ=1.919, center=<0.7201 0 0>}]
 head [{shapeType=Capsule, length=6.125, radius=4.25, rotation=<rot 90 0 0>}]
}
constraints [
 {bone1Name=dyn_hair_01, bone2Name=head, swing1LimitDegrees=8, swing2LimitDegrees=4, twistLimitDegrees=2, twistMotion=Locked}
 {bone1Name=dyn_hair_02, bone2Name=head, swing1LimitDegrees=4, swing2LimitDegrees=4, twistLimitDegrees=2, twistMotion=Locked}
 {bone1Name=dyn_hair_03, bone2Name=head, swing1LimitDegrees=6, swing2LimitDegrees=4, twistLimitDegrees=3, twistMotion=Locked}
 {bone1Name=dyn_hair_04, bone2Name=head, swing1LimitDegrees=8, swing2LimitDegrees=3, twistLimitDegrees=2, twistMotion=Locked}
 {bone1Name=dyn_ear_r_parent, bone2Name=dyn_ear_r_base, swing1LimitDegrees=10, swing2LimitDegrees=20, twistLimitDegrees=3}
 {bone1Name=dyn_ear_r_01, bone2Name=dyn_ear_r_parent, swing1LimitDegrees=10, swing2LimitDegrees=20, twistLimitDegrees=3}
 {bone1Name=dyn_ear_r_02, bone2Name=dyn_ear_r_01, swing1LimitDegrees=10, swing2LimitDegrees=20, twistLimitDegrees=3}
 {bone1Name=dyn_ear_l_02, bone2Name=dyn_ear_l_01, swing1LimitDegrees=10, swing2LimitDegrees=20, twistLimitDegrees=3}
 {bone1Name=dyn_ear_l_01, bone2Name=dyn_ear_l_parent, swing1LimitDegrees=10, swing2LimitDegrees=20, twistLimitDegrees=3}
 {bone1Name=dyn_ear_l_parent, bone2Name=dyn_ear_l_base, swing1LimitDegrees=10, swing2LimitDegrees=20, twistLimitDegrees=3}
]
