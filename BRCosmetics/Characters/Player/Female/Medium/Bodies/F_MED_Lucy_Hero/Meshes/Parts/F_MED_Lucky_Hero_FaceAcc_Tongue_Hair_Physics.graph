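# Physics asset for the tongue and hair face accessory of the Lucy Hero medium female player body.
# Four box hair bodies each constrain directly to the head capsule with tight swing limits, with twist locked on all but one.
# The tongue forms a two-box chain from a capsule-and-box base, with wider swing and minimal twist.
# No constraint links the tongue base to the head.
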PhysicsAsset /BRCosmetics/Characters/Player/Female/Medium/Bodies/F_MED_Lucy_Hero/Meshes/Parts/F_MED_Lucky_Hero_FaceAcc_Tongue_Hair_Physics
bodies {
 dyn_hair_01 [{shapeType=Box, extentX=2.3707, extentY=2.9549, extentZ=1.25, center=<1.0729 0 0>}]
 dyn_hair_02 [{shapeType=Box, extentX=3.3891, extentY=1.0279, extentZ=2.585, center=<1.3059 0 0>}]
 dyn_hair_03 [{shapeType=Box, extentX=2.7862, extentY=3.0564, extentZ=1.3489, center=<1.3063 0 0>}]
 dyn_hair_04 [{shapeType=Box, extentX=2.4759, extentY=3.3235, extentZ=1.419, center=<1.2142 0 0>}]
 dyn_tongue [{shapeType=Box, extentX=1.76, extentY=4.26, extentZ=0.8225, center=<1.067 0 0>, rotation=<rot 8.9669 0 0>}]
 dyn_tongue_01 [{shapeType=Box, extentX=3.1878, extentY=4.6073, extentZ=0.8636, center=<1.6001 0 0>}]
 dyn_tongue_base [
  {shapeType=Capsule, length=5, radius=0.75, center=<-0.8991 -0.0001 1.464>, rotation=<rot 0 0 90>}
  {shapeType=Box, extentX=1.01, extentY=1.01, extentZ=1.01, rotation=<rot 0 -90 0>}
 ]
 head [{shapeType=Capsule, length=6.125, radius=4.25, rotation=<rot 90 0 0>}]
}
constraints [
 {bone1Name=dyn_hair_01, bone2Name=head, swing1LimitDegrees=12, swing2LimitDegrees=12, twistLimitDegrees=2, twistMotion=Locked}
 {bone1Name=dyn_hair_02, bone2Name=head, swing1LimitDegrees=10, swing2LimitDegrees=10, twistLimitDegrees=2, twistMotion=Locked}
 {bone1Name=dyn_hair_03, bone2Name=head, swing1LimitDegrees=8, swing2LimitDegrees=6, twistLimitDegrees=3}
 {bone1Name=dyn_hair_04, bone2Name=head, swing1LimitDegrees=10, swing2LimitDegrees=12, twistLimitDegrees=2, twistMotion=Locked}
 {bone1Name=dyn_tongue, bone2Name=dyn_tongue_base, swing1LimitDegrees=30, swing2LimitDegrees=40, twistLimitDegrees=1}
 {bone1Name=dyn_tongue_01, bone2Name=dyn_tongue, swing1LimitDegrees=30, swing2LimitDegrees=40, twistLimitDegrees=1}
]
