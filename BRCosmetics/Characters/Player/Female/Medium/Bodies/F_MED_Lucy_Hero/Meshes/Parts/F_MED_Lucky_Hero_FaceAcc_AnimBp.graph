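# Animation Blueprint for the face accessory mesh of the Lucy/Lucky Hero cosmetic character, adding physics motion on top of the body animation.
# It copies the attached parent mesh's pose, then chains two RigidBody nodes in base-bone space: tongue/hair physics on spine_05, then face-accessory physics on spine_03.
# Each update sets ADS_Alpha from bIsTargeting, reducing only the second simulation's blend to a quarter while targeting.
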AnimBlueprint /BRCosmetics/Characters/Player/Female/Medium/Bodies/F_MED_Lucy_Hero/Meshes/Parts/F_MED_Lucky_Hero_FaceAcc_AnimBp
target_skeleton=&/F_MED_Lucky_Hero_FaceAcc_Skeleton
graphs [
 AnimGraph {nodes [
   n1 = RigidBody(ComponentPose=Local To Component(LocalPose=Copy Pose From Mesh() {bUseAttachedParent=true})) {
    bClampLinearTranslationLimitToRefPose=true
    bOverrideWorldGravity=true
    overridePhysicsAsset=&/F_MED_Lucky_Hero_FaceAcc_Tongue_Hair_Physics
    simulationSpace=BaseBoneSpace
    alphaBoolBlend {blendOption=Linear}
    alphaScaleBiasClamp {bInterpResult=true, interpSpeedDecreasing=5, interpSpeedIncreasing=5}
    baseBoneRef {boneName=spine_05}
    componentAppliedLinearAccClamp=<3000 3000 3000>
    componentLinearAccScale=<0.25 0.25 0.25>
    componentLinearVelScale=<0.25 0.25 0.25>
    overrideWorldGravity {z=-980}
   }
   n2 = RigidBody(ComponentPose=@n1, Alpha=Get ADS_Alpha() -> (ADS_Alpha:double)) {
    bClampLinearTranslationLimitToRefPose=true
    bOverrideWorldGravity=true
    overridePhysicsAsset=&/F_MED_Lucky_Hero_FaceAcc_Physics
    simulationSpace=BaseBoneSpace
    alphaBoolBlend {blendOption=Linear}
    alphaScaleBiasClamp {bInterpResult=true, interpSpeedDecreasing=5, interpSpeedIncreasing=5}
    baseBoneRef {boneName=spine_03}
    componentAppliedLinearAccClamp=<3000 3000 3000>
    componentLinearAccScale=<0.25 0.25 0.25>
    componentLinearVelScale=<0.25 0.25 0.25>
    overrideWorldGravity {z=-980}
   }
   Output Pose(Result=Component To Local(ComponentPose=@n2)) {layerGroup=None}
  ]}
 EventGraph {nodes [
   Event BlueprintUpdateAnimation()
   > Set ADS_Alpha(ADS_Alpha:double=SelectFloat(A=0.25, B=1, bPickA=Get bIsTargeting())) -> (Output_Get:double)
   TryGetPawnOwner()
  ]}
]
variables [ADS_Alpha {type=number}]
